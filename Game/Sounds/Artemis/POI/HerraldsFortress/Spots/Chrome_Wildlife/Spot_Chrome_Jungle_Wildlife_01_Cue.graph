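# Sound Cue for chrome jungle wildlife ambient spot sounds at the Herralds Fortress POI.
# A top random node picks one of three variation groups (A, B, C), each choosing a random wave under slight pitch and volume modulation.
# The root quality switch feeds this chain to its first two levels and an empty modulator to the last two, leaving those silent.
# Playback uses the outdoor ambient sound class and wide one-shot attenuation.
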SoundCue /Game/Sounds/Artemis/POI/HerraldsFortress/Spots/Chrome_Wildlife/Spot_Chrome_Jungle_Wildlife_01_Cue
root=QualityLevel_0
nodes [
 {id=Modulator_0, children [Random_2], properties {pitchMax=1.1, pitchMin=1, volumeMax=1.2, volumeMin=1}}
 {id=Modulator_1, children [Random_1], properties {pitchMax=1.1, pitchMin=1, volumeMax=1.2, volumeMin=1}}
 {id=Modulator_2, children [null], properties {pitchMax=1.1, volumeMax=1.1, volumeMin=0.1}}
 {id=Modulator_3, children [Random_3], properties {pitchMax=1.1, pitchMin=1, volumeMax=1.2, volumeMin=1}}
 {id=QualityLevel_0, children [Random_4,Random_4,Modulator_2,Modulator_2]}
 {
  id=Random_1
  children [WavePlayer_8, WavePlayer_9, WavePlayer_10, WavePlayer_11, WavePlayer_12, WavePlayer_13]
  properties {weights [1,1,1,1,1,1]}
 }
 {id=Random_2, children [WavePlayer_14, WavePlayer_15, WavePlayer_16, WavePlayer_17, WavePlayer_18], properties {weights [1,1,1,1,1]}}
 {
  id=Random_3
  children [WavePlayer_19, WavePlayer_20, WavePlayer_21, WavePlayer_22, WavePlayer_23, WavePlayer_24, WavePlayer_25, WavePlayer_26]
  properties {weights [1,1,1,1,1,1,1,1]}
 }
 {id=Random_4, children [Modulator_1,Modulator_0,Modulator_3], properties {weights [1,1,1]}}
 {
  id=WavePlayer_10
  properties {soundWaveAssetPtr=&/Spot_Chrome_Jungle_Wildlife_A_03}
 }
 {
  id=WavePlayer_11
  properties {soundWaveAssetPtr=&/Spot_Chrome_Jungle_Wildlife_A_04}
 }
 {
  id=WavePlayer_12
  properties {soundWaveAssetPtr=&/Spot_Chrome_Jungle_Wildlife_A_05}
 }
 {
  id=WavePlayer_13
  properties {soundWaveAssetPtr=&/Spot_Chrome_Jungle_Wildlife_A_06}
 }
 {
  id=WavePlayer_14
  properties {soundWaveAssetPtr=&/Spot_Chrome_Jungle_Wildlife_B_01}
 }
 {
  id=WavePlayer_15
  properties {soundWaveAssetPtr=&/Spot_Chrome_Jungle_Wildlife_B_02}
 }
 {
  id=WavePlayer_16
  properties {soundWaveAssetPtr=&/Spot_Chrome_Jungle_Wildlife_B_03}
 }
 {
  id=WavePlayer_17
  properties {soundWaveAssetPtr=&/Spot_Chrome_Jungle_Wildlife_B_04}
 }
 {
  id=WavePlayer_18
  properties {soundWaveAssetPtr=&/Spot_Chrome_Jungle_Wildlife_B_05}
 }
 {
  id=WavePlayer_19
  properties {soundWaveAssetPtr=&/Spot_Chrome_Jungle_Wildlife_C_01}
 }
 {
  id=WavePlayer_20
  properties {soundWaveAssetPtr=&/Spot_Chrome_Jungle_Wildlife_C_02}
 }
 {
  id=WavePlayer_21
  properties {soundWaveAssetPtr=&/Spot_Chrome_Jungle_Wildlife_C_03}
 }
 {
  id=WavePlayer_22
  properties {soundWaveAssetPtr=&/Spot_Chrome_Jungle_Wildlife_C_04}
 }
 {
  id=WavePlayer_23
  properties {soundWaveAssetPtr=&/Spot_Chrome_Jungle_Wildlife_C_05}
 }
 {
  id=WavePlayer_24
  properties {soundWaveAssetPtr=&/Spot_Chrome_Jungle_Wildlife_C_06}
 }
 {
  id=WavePlayer_25
  properties {soundWaveAssetPtr=&/Spot_Chrome_Jungle_Wildlife_C_07}
 }
 {
  id=WavePlayer_26
  properties {soundWaveAssetPtr=&/Spot_Chrome_Jungle_Wildlife_C_08}
 }
 {
  id=WavePlayer_8
  properties {soundWaveAssetPtr=&/Spot_Chrome_Jungle_Wildlife_A_01}
 }
 {
  id=WavePlayer_9
  properties {soundWaveAssetPtr=&/Spot_Chrome_Jungle_Wildlife_A_02}
 }
]
properties {
 attenuationSettings=/Game/Sounds/Attenuation/Fort_TileAmbient_OneShot_Wider_Att
 bEnableBusSends=false
 maxDistance=8000
 soundClassObject=&Fort_Ambient_Outdoors_Athena
 volumeMultiplier=0.802
 preEffectBusSends [{}]
}
