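# Sound cue for a Chrome jungle wildlife spot ambience at the Herralds Fortress POI, using outdoor ambient class and one-shot tile attenuation.
# An equally weighted random picks one of four wildlife groups (D, E, F, G), each randomizing among its own waves with slight pitch and volume variation.
# A quality-level root routes the first two slots to that random tree, while the last two share a childless modulator and play nothing.
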SoundCue /Game/Sounds/Artemis/POI/HerraldsFortress/Spots/Chrome_Wildlife/Spot_Chrome_Jungle_Wildlife_02_Cue
root=QualityLevel_0
nodes [
 {id=Modulator_0, children [Random_5], properties {pitchMax=1.1, pitchMin=1, volumeMax=1.2, volumeMin=1}}
 {id=Modulator_1, children [Random_0], properties {pitchMax=1.1, pitchMin=1, volumeMax=1.2, volumeMin=1}}
 {id=Modulator_2, children [null], properties {pitchMax=1.1, volumeMax=1.1, volumeMin=0.1}}
 {id=Modulator_3, children [Random_6], properties {pitchMax=1.1, pitchMin=1, volumeMax=1.2, volumeMin=1}}
 {id=Modulator_4, children [Random_7], properties {pitchMax=1.1, pitchMin=1, volumeMax=1.2, volumeMin=1}}
 {id=QualityLevel_0, children [Random_4,Random_4,Modulator_2,Modulator_2]}
 {
  id=Random_0
  children [WavePlayer_0,WavePlayer_1,WavePlayer_2,WavePlayer_3,WavePlayer_4,WavePlayer_5,WavePlayer_6]
  properties {weights [1,1,1,1,1,1,1]}
 }
 {id=Random_4, children [Modulator_1,Modulator_0,Modulator_3,Modulator_4], properties {weights [1,1,1,1]}}
 {id=Random_5, children [WavePlayer_7, WavePlayer_27, WavePlayer_28, WavePlayer_29], properties {weights [1,1,1,1]}}
 {
  id=Random_6
  children [WavePlayer_30, WavePlayer_31, WavePlayer_32, WavePlayer_33, WavePlayer_34, WavePlayer_35]
  properties {weights [1,1,1,1,1,1]}
 }
 {
  id=Random_7
  children [WavePlayer_36, WavePlayer_37, WavePlayer_38, WavePlayer_39, WavePlayer_40, WavePlayer_41, WavePlayer_42, WavePlayer_43]
  properties {weights [1,1,1,1,1,1,1,1]}
 }
 {
  id=WavePlayer_0
  properties {soundWaveAssetPtr=&/Spot_Chrome_Jungle_Wildlife_D_01}
 }
 {
  id=WavePlayer_1
  properties {soundWaveAssetPtr=&/Spot_Chrome_Jungle_Wildlife_D_02}
 }
 {
  id=WavePlayer_2
  properties {soundWaveAssetPtr=&/Spot_Chrome_Jungle_Wildlife_D_03}
 }
 {
  id=WavePlayer_27
  properties {soundWaveAssetPtr=&/Spot_Chrome_Jungle_Wildlife_E_02}
 }
 {
  id=WavePlayer_28
  properties {soundWaveAssetPtr=&/Spot_Chrome_Jungle_Wildlife_E_03}
 }
 {
  id=WavePlayer_29
  properties {soundWaveAssetPtr=&/Spot_Chrome_Jungle_Wildlife_E_04}
 }
 {
  id=WavePlayer_3
  properties {soundWaveAssetPtr=&/Spot_Chrome_Jungle_Wildlife_D_04}
 }
 {
  id=WavePlayer_30
  properties {soundWaveAssetPtr=&/Spot_Chrome_Jungle_Wildlife_F_01}
 }
 {
  id=WavePlayer_31
  properties {soundWaveAssetPtr=&/Spot_Chrome_Jungle_Wildlife_F_02}
 }
 {
  id=WavePlayer_32
  properties {soundWaveAssetPtr=&/Spot_Chrome_Jungle_Wildlife_F_03}
 }
 {
  id=WavePlayer_33
  properties {soundWaveAssetPtr=&/Spot_Chrome_Jungle_Wildlife_F_04}
 }
 {
  id=WavePlayer_34
  properties {soundWaveAssetPtr=&/Spot_Chrome_Jungle_Wildlife_F_05}
 }
 {
  id=WavePlayer_35
  properties {soundWaveAssetPtr=&/Spot_Chrome_Jungle_Wildlife_F_06}
 }
 {
  id=WavePlayer_36
  properties {soundWaveAssetPtr=&/Spot_Chrome_Jungle_Wildlife_G_01}
 }
 {
  id=WavePlayer_37
  properties {soundWaveAssetPtr=&/Spot_Chrome_Jungle_Wildlife_G_02}
 }
 {
  id=WavePlayer_38
  properties {soundWaveAssetPtr=&/Spot_Chrome_Jungle_Wildlife_G_03}
 }
 {
  id=WavePlayer_39
  properties {soundWaveAssetPtr=&/Spot_Chrome_Jungle_Wildlife_G_04}
 }
 {
  id=WavePlayer_4
  properties {soundWaveAssetPtr=&/Spot_Chrome_Jungle_Wildlife_D_05}
 }
 {
  id=WavePlayer_40
  properties {soundWaveAssetPtr=&/Spot_Chrome_Jungle_Wildlife_G_05}
 }
 {
  id=WavePlayer_41
  properties {soundWaveAssetPtr=&/Spot_Chrome_Jungle_Wildlife_G_06}
 }
 {
  id=WavePlayer_42
  properties {soundWaveAssetPtr=&/Spot_Chrome_Jungle_Wildlife_G_07}
 }
 {
  id=WavePlayer_43
  properties {soundWaveAssetPtr=&/Spot_Chrome_Jungle_Wildlife_G_08}
 }
 {
  id=WavePlayer_5
  properties {soundWaveAssetPtr=&/Spot_Chrome_Jungle_Wildlife_D_06}
 }
 {
  id=WavePlayer_6
  properties {soundWaveAssetPtr=&/Spot_Chrome_Jungle_Wildlife_D_07}
 }
 {
  id=WavePlayer_7
  properties {soundWaveAssetPtr=&/Spot_Chrome_Jungle_Wildlife_E_01}
 }
]
properties {
 attenuationSettings=/Game/Sounds/Attenuation/Fort_TileAmbient_OneShot_Wider_Att
 bEnableBusSends=false
 maxDistance=8000
 soundClassObject=&Fort_Ambient_Outdoors_Athena
 volumeMultiplier=0.802
 preEffectBusSends [{}]
}
